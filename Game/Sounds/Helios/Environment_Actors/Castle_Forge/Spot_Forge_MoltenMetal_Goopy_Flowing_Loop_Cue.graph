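# Looping ambient cue for flowing molten metal at the Castle Forge, built from Brimstone Waterfall standard and low-end loops.
# Each loop gets a three-second volume fade-in and random pitch modulation.
# A mixer sums three attenuated branches: the standard loop under Main Flow, and the shared low-end chain under Main Flow and Rumble Core settings.
# It overrides concurrency to two voices and feeds an exterior ambient bus and reverb submix.
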SoundCue /Game/Sounds/Helios/Environment_Actors/Castle_Forge/Spot_Forge_MoltenMetal_Goopy_Flowing_Loop_Cue
root=Mixer_1
nodes [
 {
  id=Attenuation_0
  children [Modulator_0]
  properties {attenuationSettings=&/Spot_Forge_MoltenMetal_Main_Flow_Att}
 }
 {
  id=Attenuation_1
  children [Modulator_2]
  properties {attenuationSettings=&/Spot_Forge_MoltenMetal_Main_Flow_Rumble_Core_Att}
 }
 {
  id=Attenuation_2
  children [Modulator_2]
  properties {attenuationSettings=&Spot_Forge_MoltenMetal_Main_Flow_Att}
 }
 {
  id=Enveloper_0
  children [Mixer_2]
  properties {
   pitchCurve {editorCurveData {keys [{time=0}]}}
   volumeCurve {editorCurveData {keys [<at -0.0001 0 RCIM_Cubic>, {arriveTangent=0.4839, interpMode=RCIM_Cubic, leaveTangent=0.4839, tangentMode=RCTM_User, time=3, value=1}]}}
  }
 }
 {
  id=Enveloper_1
  children [WavePlayer_1]
  properties {
   pitchCurve {editorCurveData {keys [{time=0}]}}
   volumeCurve {editorCurveData {keys [<at -0.0001 0 RCIM_Cubic>, {arriveTangent=0.4839, interpMode=RCIM_Cubic, leaveTangent=0.4839, tangentMode=RCTM_User, time=3, value=1}]}}
  }
 }
 {id=Mixer_1, children [Attenuation_0, Attenuation_2, Attenuation_1], properties {inputVolume [2,1,2]}}
 {id=Mixer_2, children [WavePlayer_2,null], properties {inputVolume [1,1]}}
 {id=Modulator_0, children [Enveloper_1], properties {pitchMax=1.1, pitchMin=0.9, volumeMax=1, volumeMin=1}}
 {id=Modulator_2, children [Enveloper_0], properties {pitchMax=1.1, pitchMin=1, volumeMax=1, volumeMin=1}}
 {
  id=WavePlayer_1
  properties {bLooping=true, soundWaveAssetPtr=/Game/Sounds/Helios/Environment_Actors/BrimstoneWaterfall/Spot_Brimstone_Waterfall_Standard_Loop}
 }
 {
  id=WavePlayer_2
  properties {bLooping=true, soundWaveAssetPtr=/Game/Sounds/Helios/Environment_Actors/BrimstoneWaterfall/Spot_Brimstone_Waterfall_Low_End}
 }
]
properties {
 bOverrideConcurrency=true
 maxDistance=850
 priority=10
 soundClassObject=&Fort_Ambient_Indoors_Athena
 volumeMultiplier=0.752
 busSends [{audioBus=null, minSendLevel=1, sendLevel=1, soundSourceBus=&AmbientAudio_Exterior_SourceBus}]
 concurrencyOverrides {maxCount=2, voiceStealReleaseTime=2}
 soundSubmixSends [{minSendLevel=1, soundSubmix=&Fortnite_Reverb_Default}]
}
