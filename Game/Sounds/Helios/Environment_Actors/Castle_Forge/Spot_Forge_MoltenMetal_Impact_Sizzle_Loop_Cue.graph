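# Looping ambient Sound Cue for molten metal impact sizzle in the Castle Forge environment.
# Mixes a boosted brimstone waterfall loop, a random pitched-down water thickener loop (both borrowed from Mephisto), and a random sizzle loop with slow volume oscillation.
# Modulators randomize pitch on the sizzle and overall mix, with all quality levels sharing one chain.
# Uses waterfall impact attenuation, limited concurrency, reverb and an exterior ambient bus send.
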SoundCue /Game/Sounds/Helios/Environment_Actors/Castle_Forge/Spot_Forge_MoltenMetal_Impact_Sizzle_Loop_Cue
root=QualityLevel_0
nodes [
 {id=Mixer_0, children [Modulator_2,WavePlayer_0,Modulator_0], properties {inputVolume [1,1.5,1]}}
 {id=Modulator_0, children [Oscillator_0], properties {pitchMax=1.1, pitchMin=0.9, volumeMax=1, volumeMin=1}}
 {id=Modulator_1, children [Mixer_0], properties {pitchMax=1.1, pitchMin=0.9, volumeMax=1, volumeMin=1}}
 {id=Modulator_2, children [Random_0], properties {pitchMax=0.5, pitchMin=0.5}}
 {
  id=Oscillator_0
  children [Random_1]
  properties {amplitudeMax=0.3, amplitudeMin=0.2, bModulateVolume=true, centerMax=1, centerMin=1, frequencyMax=0.1, frequencyMin=0.1}
 }
 {id=QualityLevel_0, children [Modulator_1,Modulator_1,Modulator_1,Modulator_1]}
 {id=Random_0, children [WavePlayer_3,WavePlayer_4,WavePlayer_5], properties {weights [1,1,1]}}
 {id=Random_1, children [WavePlayer_1,WavePlayer_2], properties {weights [1,1]}}
 {
  id=WavePlayer_0
  properties {bLooping=true, soundWaveAssetPtr=/Game/Sounds/Helios/Environment_Actors/Mephisto_Hades/Spot_Mephisto_Brimstone_Waterfall_Bottom_Loop}
 }
 {
  id=WavePlayer_1
  properties {bLooping=true, soundWaveAssetPtr=&/Spot_Forge_MoltenMetal_Sizzle_01}
 }
 {
  id=WavePlayer_2
  properties {bLooping=true, soundWaveAssetPtr=&/Spot_Forge_MoltenMetal_Sizzle_02}
 }
 {
  id=WavePlayer_3
  properties {bLooping=true, soundWaveAssetPtr=/Game/Sounds/Helios/Environment_Actors/Mephisto_Hades/Spot_Mephisto_Water_Flowing_Thickener_Loop_01}
 }
 {
  id=WavePlayer_4
  properties {bLooping=true, soundWaveAssetPtr=/Game/Sounds/Helios/Environment_Actors/Mephisto_Hades/Spot_Mephisto_Water_Flowing_Thickener_Loop_02}
 }
 {
  id=WavePlayer_5
  properties {bLooping=true, soundWaveAssetPtr=/Game/Sounds/Helios/Environment_Actors/Mephisto_Hades/Spot_Mephisto_Water_Flowing_Thickener_Loop_03}
 }
]
properties {
 attenuationSettings=/Game/Sounds/Helios/Environment_Actors/BrimstoneWaterfall/Spot_Water_Waterfall_Goopy_Impact_Heavy_Occ_Att
 bOverrideConcurrency=true
 maxDistance=2000
 pitchMultiplier=1.2
 priority=10
 soundClassObject=&Fort_Ambient_Indoors_Athena
 volumeMultiplier=0.752
 busSends [{audioBus=null, minSendLevel=1, sendLevel=1, soundSourceBus=&AmbientAudio_Exterior_SourceBus}]
 concurrencyOverrides {maxCount=2, voiceStealReleaseTime=2}
 soundSubmixSends [{minSendLevel=1, soundSubmix=&Fortnite_Reverb_Default}]
}
