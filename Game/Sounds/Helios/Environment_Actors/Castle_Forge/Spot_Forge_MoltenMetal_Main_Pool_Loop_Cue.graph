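# Looping spot ambience cue for the molten metal main pool in the Castle Forge, built from reused Mephisto brimstone waterfall waves.
# A standard waterfall loop plays over a half-volume bed of low-end and rumble bass loops.
# Each layer has an envelope fade-in and random pitch modulation, and the whole cue is pitched down.
# It uses occlusion-named waterfall attenuation, an indoor ambient class, a reverb send and a priority-based concurrency cap.
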SoundCue /Game/Sounds/Helios/Environment_Actors/Castle_Forge/Spot_Forge_MoltenMetal_Main_Pool_Loop_Cue
root=Mixer_0
nodes [
 {
  id=Enveloper_0
  children [Mixer_1]
  properties {
   pitchCurve {editorCurveData {keys [{time=0}]}}
   volumeCurve {editorCurveData {keys [<at -0.0001 0 RCIM_Cubic>, {arriveTangent=0.4839, interpMode=RCIM_Cubic, leaveTangent=0.4839, tangentMode=RCTM_User, time=3, value=1}]}}
  }
 }
 {
  id=Enveloper_1
  children [WavePlayer_3]
  properties {
   pitchCurve {editorCurveData {keys [{time=0}]}}
   volumeCurve {editorCurveData {keys [<at -0.0001 0 RCIM_Cubic>, {arriveTangent=0.4839, interpMode=RCIM_Cubic, leaveTangent=0.4839, tangentMode=RCTM_User, time=3, value=1}]}}
  }
 }
 {id=Mixer_0, children [Modulator_1,Modulator_0], properties {inputVolume [1,0.5]}}
 {id=Mixer_1, children [WavePlayer_0,WavePlayer_2], properties {inputVolume [1,1]}}
 {id=Modulator_0, children [Enveloper_0], properties {pitchMax=1.1, pitchMin=1, volumeMax=1, volumeMin=1}}
 {id=Modulator_1, children [Enveloper_1], properties {pitchMax=1.1, pitchMin=0.9, volumeMax=1, volumeMin=1}}
 {
  id=WavePlayer_0
  properties {bLooping=true, soundWaveAssetPtr=/Game/Sounds/Helios/Environment_Actors/Mephisto_Hades/Spot_Mephisto_Brimstone_Waterfall_Low_End}
 }
 {
  id=WavePlayer_2
  properties {bLooping=true, soundWaveAssetPtr=/Game/Sounds/Helios/Environment_Actors/Mephisto_Hades/Spot_Mephisto_Brimstone_Waterfall_Rumble_Bass_Loop}
 }
 {
  id=WavePlayer_3
  properties {bLooping=true, soundWaveAssetPtr=/Game/Sounds/Helios/Environment_Actors/Mephisto_Hades/Spot_Mephisto_Brimstone_Waterfall_Standard_Loop}
 }
]
properties {
 attenuationSettings=/Game/Sounds/Helios/Environment_Actors/BrimstoneWaterfall/Spot_BrimstoneWaterfall_MainFlow_Occ_Att
 bOverrideConcurrency=true
 maxDistance=3850
 pitchMultiplier=0.8
 priority=10
 soundClassObject=&Fort_Ambient_Indoors_Athena
 volumeMultiplier=0.702
 concurrencyOverrides {maxCount=2, resolutionRule=StopLowestPriority}
 soundSubmixSends [{minSendLevel=1, soundSubmix=&Fortnite_Reverb_Default}]
}
