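# Looping ambient sound cue for the top of a molten metal fall in the Castle Forge environment.
# It loops a reused Mephisto brimstone waterfall wave through a modulator that slightly randomizes pitch, with every quality level sharing that branch.
# Playback uses molten metal fall attenuation, the Athena ambient sound class and a default reverb send.
# A concurrency override allows one instance at a time, with a release time on voice stealing.
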SoundCue /Game/Sounds/Helios/Environment_Actors/Castle_Forge/Spot_Forge_MoltenMetal_Top_Loop_Cue
root=QualityLevel_0
nodes [
 {id=Modulator_1, children [WavePlayer_0], properties {pitchMax=1.1, pitchMin=0.9, volumeMax=1, volumeMin=1}}
 {id=QualityLevel_0, children [Modulator_1,Modulator_1,Modulator_1,Modulator_1]}
 {
  id=WavePlayer_0
  properties {bLooping=true, soundWaveAssetPtr=/Game/Sounds/Helios/Environment_Actors/Mephisto_Hades/Spot_Mephisto_Brimstone_Waterfall_Top_Loop}
 }
]
properties {
 attenuationSettings=&/Spot_Forge_MoltenMetal_Fall_Top_Att
 bOverrideConcurrency=true
 maxDistance=1750
 pitchMultiplier=1.1
 priority=10
 soundClassObject=&Fort_Ambient_Athena
 volumeMultiplier=0.401
 concurrencyOverrides {maxCount=1, voiceStealReleaseTime=3}
 soundSubmixSends [{minSendLevel=1, soundSubmix=&Fortnite_Reverb_Default}]
}
